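# Player animation blueprint for the HoneyFiber Incident item, layering a throw aim offset onto the Fortnite player skeleton through Item Override layers.
# Each update copies the HoneyFiber_Equipped anim curve into the aim offset's alpha.
# ItemFullBodyAimOffsets applies that offset to the pre-aim-offset pose using the layer's aim yaw and pitch.
# Other layers are unwired, including an upper-body aim offset, except the zipline-start additive, which returns a reference pose.
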AnimBlueprint /HoneyFiber_Incident/HoneyFiber_Incident_PlayerAnimBP
target_skeleton=&Fortnite_M_Avg_Player_Skeleton
graphs [
 AnimGraph {nodes [Output Pose()]}
 EventGraph {nodes [
   Event BlueprintUpdateAnimation()
   > Set HoneyFiber_AO_Alpha(HoneyFiber_AO_Alpha:double=GetCurveValue(CurveName=HoneyFiber_Equipped)) -> (Output_Get:double)
   TryGetPawnOwner()
  ]}
 ItemAimOffset {nodes [
   Input Pose!() -> (Pose, AimOffsetAlpha:double, Yaw:double, Pitch:double) {name=InPose_AimOffset}
   Input Pose!() -> (Pose) {name=InPose_PreAimOffsets}
   Output Pose() {layerGroup=Item Override}
  ]}
 ItemFallAdditive {nodes [
   Input Pose() {name=InPoseItemFallAdditive}
   Output Pose() {layerGroup=Item Override}
  ]}
 ItemFallLandAdditive {nodes [
   Input Pose() {name=InPoseItemFallLandAdditive}
   Output Pose() {layerGroup=Item Override}
  ]}
 ItemFallLevel2Override {nodes [
   Input Pose() {name=InPose_FallLevel2}
   Output Pose() {layerGroup=Item Override}
  ]}
 ItemFinalPoseOverride {nodes [
   Input Pose() {name=InFinalPose}
   Output Pose() {layerGroup=Item Override}
  ]}
 ItemFullBodyAimOffsets {nodes [
   n1 = Input Pose!() -> (Pose, FullBody_AimOffsetAlpha:double, Aim_Yaw:double, Aim_Pitch:double) {name=InPoseFullBodyAimOffsets}
   Output Pose(Result=AimOffset Player 'HoneyFiber_Incident_Throw_AO'!(
     X:float=@n1.Aim_Yaw
     Y:float=@n1.Aim_Pitch
     BasePose:Pose Link Structure=Input Pose!() -> (Pose) {name=InPosePreFullBodyAimOffsets}
     Alpha:float=Get HoneyFiber_AO_Alpha() -> (HoneyFiber_AO_Alpha:double)
    ) -> (Pose:Pose Link Structure) {blendSpace=/HoneyFiber_Incident/Assets/Anims/AO/HoneyFiber_Incident_Throw_AO}) {layerGroup=Item Override}
  ]}
 ItemFullBodyOverride {nodes [
   Input Pose() {name=InPoseFullBody}
   Output Pose() {layerGroup=Item Override}
  ]}
 ItemIdleOverride {nodes [
   Input Pose() {name=InPose_Idle}
   Output Pose() {layerGroup=Item Override}
  ]}
 ItemJetPackJumpAdditive {nodes [Output Pose() {layerGroup=Item Override}]}
 ItemJetPackStartAdditive {nodes [Output Pose() {layerGroup=Item Override}]}
 ItemJumpLoopAdditive {nodes [
   Input Pose() {name=InPoseItemJumpLoopAdditive}
   Output Pose() {layerGroup=Item Override}
  ]}
 ItemJumpOverride {nodes [
   Input Pose() {name=InPoseJump}
   Output Pose() {layerGroup=Item Override}
  ]}
 ItemJumpUpAdditive {nodes [
   Input Pose() {name=InPoseItemJumpUpAdditive}
   Output Pose() {layerGroup=Item Override}
  ]}
 ItemPreAimOffsets {nodes [
   Input Pose() {name=InPosePreAimOffsets}
   Output Pose() {layerGroup=Item Override}
  ]}
 ItemPreIK {nodes [
   Input Pose() {name=InPose_PreIK}
   Output Pose() {layerGroup=Item Override}
  ]}
 ItemRightHandIK {nodes [
   Input Pose() {name=InPosePreRightHandIK}
   Input Pose() {name=InPosePostRightHandIK}
   Output Pose() {layerGroup=Item Override}
  ]}
 ItemSkeletalControl {nodes [
   Input Pose()
   Output Pose() {layerGroup=Item Override}
  ]}
 ItemSwimJumpFallLoop {nodes [Output Pose() {layerGroup=Item Override}]}
 ItemSwimJumpSurfaceEnd {nodes [Output Pose() {layerGroup=Item Override}]}
 ItemSwimJumpSurfaceLoop {nodes [Output Pose() {layerGroup=Item Override}]}
 ItemUpperBody {nodes [
   Input Pose!() -> (Pose, InputParam:Fort Anim Input Adjusted Aim Structure) {name=InPoseUpperBody}
   Output Pose() {layerGroup=Item Override}
   AimOffset Player 'HoneyFiber_Incident_Throw_AO'(X=Get AimYaw() -> (AimYaw:float), Y=Get AimPitch(), Alpha=1) {blendSpace=&HoneyFiber_Incident_Throw_AO}
  ]}
 ItemZipLineStartAdditive {nodes [Output Pose(Result=Local Space Ref Pose()) {layerGroup=Item Override}]}
]
variables [
 HoneyFiber_AO_Alpha {type=number}
 HoneyFiber_Equipped {type=boolean}
]
